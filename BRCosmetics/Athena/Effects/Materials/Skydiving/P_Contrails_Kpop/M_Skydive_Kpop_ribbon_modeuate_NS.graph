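# Unlit two-sided Modulate material for K-pop skydive contrail ribbons and sprites.
# A line texture's UVs are warped by animated sine waves using world height, dynamic offset, speed and reverse inputs, plus normal-map detail outside mobile.
# The thresholded line mask, with optional foot mask, takes ramp or particle color and whitens elsewhere so the background stays unchanged.
# A separate branch builds a Z offset at the low-U end for near-vertical camera views.
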Material /BRCosmetics/Athena/Effects/Materials/Skydiving/P_Contrails_Kpop/M_Skydive_Kpop_ribbon_modeuate_NS
graph {
 properties {
  bUsedWithNiagaraRibbons=true
  bUsedWithNiagaraSprites=true
  bUsedWithParticleSprites=true
  blendMode=BLEND_Modulate
  pixelDepthOffsetMode=PDOM_Legacy
  shadingModel=MSM_Unlit
  translucencyPass=MTP_BeforeDOF
  twoSided=true
  usedShadingModels=Unlit
 }
 nodes [
  n1 = ParticleColor()
  n2 = Multiply(A=TextureCoordinate(), B=Constant2Vector() {g=1, r=0.2})
  n3 = Time()
  n4 = TextureSample(UVs=MaterialFunctionCall(
    UVs=Panner(Coordinate=@n2, Time=@n3) {speedY=0.2}
    Rotation Center=Constant2Vector() {g=0.5, r=0.5}
    "Rotation Angle (0-1)"=ScalarParameter() {defaultValue=0.25, parameterName=TexRotation}
   ) {materialFunction=&CustomRotator}.Rotated Values) {texture=/BRCosmetics/Athena/Effects/Textures/Skydive/P_Contrail_Kpop/T_Contrail_Kpop_Ramp}
  n5 = DynamicParameter() {paramNames ["offset(0-1)", speed, "reverse(0,1)", Param4]}
  n6 = TextureCoordinate()
  n7 = ComponentMask(None=@n6) {r=true}
  n8 = AppendVector(A=Abs(None=Subtract(A=@n5."reverse(0,1)", B=@n7)), B=ComponentMask(None=@n6) {g=true})
  n9 = ComponentMask(None=@n8) {r=true}
  n10 = Divide(A=OneMinus(None=@n9)) {constB=4}
  n11 = Add(A=@n8, B=AppendVector(A=Constant(), B=@n5."offset(0-1)"))
  n12 = WorldPosition()
  n13 = Constant() {r=-1618}
  n14 = Add(A=Multiply(A=Time(), B=@n5.speed) {constB=10}, B=Divide(A=ComponentMask(None=@n12.XYZ) {b=true}, B=Divide(A=@n13) {constB=4}) {constB=1618})
  n15 = Frac(None=ComponentMask(None=Panner(Coordinate=@n11, Time=@n14) {speedX=0.1, speedY=0.2}) {r=true})
  n16 = ScalarParameter() {defaultValue=1, group=Distort_map, parameterName=Disort_frequency, sortPriority=4}
  n17 = Panner(Coordinate=Add(A=@n11, B=Constant2Vector() {r=0.2}), Time=@n14) {speedY=0.3}
  n18 = Multiply(A=Frac(None=ComponentMask(None=@n17) {r=true}), B=@n16)
  n19 = Add(A=Sine(None=Multiply(A=@n15, B=@n16)) {period=0.5}, B=Multiply(A=Cosine(None=@n18)) {constB=0.1})
  n20 = ScalarParameter() {defaultValue=0.5, group=Distort_map, parameterName="distort_channel(R-0,G-0.5,B-1)", sortPriority=2}
  n21 = Constant() {r=0.5}
  n22 = VectorParameter() {group=Distort_map, parameterName="distortUV_(R-U,G-V)", sortPriority=1, defaultValue=<rgba 0.1 0.3 0 1>}
  n23 = Multiply(A=@n8, B=ComponentMask(None=@n22.RGB) {g=true, r=true})
  n24 = Time()
  n25 = TextureSampleParameter2D(UVs=MaterialFunctionCall(
    UVs=Panner(Coordinate=@n23, Time=@n24) {speedX=0.2}
    Rotation Center=Constant2Vector() {g=0.5, r=0.5}
    "Rotation Angle (0-1)"=ScalarParameter() {defaultValue=0.25, parameterName=TexRotation}
   ) {materialFunction=&CustomRotator}.Rotated Values) {
   group=Distort_map
   parameterName=Detail_disort_map
   samplerType=SAMPLERTYPE_Normal
   texture=/Game/Global/Textures/Default/Tileable_Patterns/1Normal
  }
  n26 = Add(
   A=@n19
   B=LinearInterpolate(
    B=ScalarParameter() {defaultValue=2, group=Distort_map, parameterName=Distort_lerp, sortPriority=3}
    Alpha=If(A=@n20, B=@n21, "A > B"=@n25.B, "A == B"=@n25.B, "A < B"=@n25.B)
   ) {constB=2}
  ) {constB=0}
  n27 = Multiply(
   A=Multiply(A=Sine(None=@n10), B=ShadingPathSwitch(Default=@n26, Mobile=@n19) {desc=Migrated from FeatureLevelSwitch})
   B=ScalarParameter() {defaultValue=0.15, group=Distort_map, parameterName=Disort_amplitude, sortPriority=5}
  ) {constB=0.15}
  n28 = TextureSampleParameter2D(UVs=MaterialFunctionCall(
    UVs=Add(A=@n8, B=@n27) {constB=0}
    Rotation Center=Constant2Vector() {g=0.5, r=0.5}
    "Rotation Angle (0-1)"=ScalarParameter() {defaultValue=0.25, parameterName=TexRotation}
   ) {materialFunction=&CustomRotator}.Rotated Values) {group=Elements, parameterName=Line, texture=/BRCosmetics/Athena/Effects/Textures/Skydive/P_Contrail_Kpop/T_Contrail_Kpop_elements}
  n29 = Multiply(A=OneMinus(None=@n9), B=@n9)
  n30 = Multiply(
   A=If(
    A=ScalarParameter() {group=Elements, parameterName="channel(R-0,G-0.5,B-1)", sortPriority=1}
    B=Constant() {r=0.5}
    "A > B"=@n28.B
    "A == B"=@n28.B
    "A < B"=@n28.B
   )
   B=Power(Base=LinearInterpolate(Alpha=@n29) {constB=4})
  )
  n31 = LinearInterpolate(Alpha=Clamp(None=LinearInterpolate(Alpha=@n9) {constA=-0.5, constB=0.5})) {constB=2}
  n32 = StaticSwitch(True=Multiply(A=@n30, B=Multiply(A=@n31, B=@n31)), False=@n30, Value=StaticBoolParameter() {parameterName=foot_mask})
  n33 = Constant()
  n34 = Multiply(
   A=StaticSwitch(True=@n1.RGB, False=Multiply(A=@n4.RGB, B=@n1.RGB), Value=StaticBoolParameter() {parameterName=use_paricle_color})
   B=If(A=Clamp(None=@n32), B=@n33, "A > B"=Constant() {r=1}, "A == B"=@n33, "A < B"=@n33)
  )
  n35 = Constant() {r=1}
  n36 = If(A=Subtract(A=Constant() {r=1}, B=@n32), B=@n35, "A > B"=@n35, "A == B"=@n35, "A < B"=Constant())
  n37 = OneMinus(None=LinearInterpolate(Alpha=@n32) {constB=10})
  Clamp(None=Add(A=Add(A=@n34, B=@n36), B=Clamp(None=@n37)))
  n38 = OneMinus(None=ComponentMask(None=TextureCoordinate()) {r=true})
  n39 = Constant() {r=0.9}
  n40 = Constant() {r=1}
  n41 = Constant()
  n42 = ObjectPositionWS()
  n43 = AppendVector(A=Constant2Vector(), B=ComponentMask(None=@n42) {b=true})
  n44 = Subtract(A=CameraPositionWS(), B=@n42)
  n45 = Abs(None=DotProduct(A=Normalize(VectorInput=@n43), B=Normalize(VectorInput=@n44)))
  n46 = Constant() {r=0.9}
  n47 = Constant() {r=30}
  n48 = Constant()
  Multiply(
   A=Multiply(A=If(A=@n38, B=@n39, "A > B"=@n40, "A == B"=@n41, "A < B"=@n41))
   B=AppendVector(A=Constant2Vector(), B=If(A=@n45, B=@n46, "A > B"=@n47, "A == B"=@n48, "A < B"=@n48))
  )
 ]
}
